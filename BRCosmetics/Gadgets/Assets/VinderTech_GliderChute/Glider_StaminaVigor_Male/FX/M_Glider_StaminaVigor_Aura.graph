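# Unlit translucent Niagara sprite material for the flame-like aura effect on the Stamina Vigor male glider.
# Emissive blends normal-distorted panning noise with a masked dark-to-bright gradient, particle color, time-of-day tint and night desaturation.
# Opacity shapes a flame texture with UV and 3D sphere masks, optional disturbance and mask texture, and view-angle fade.
# Dynamic parameters drive sphere radius, edge fade and a vertical appear wipe; Forward and Mobile paths skip depth fade.
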Material /BRCosmetics/Gadgets/Assets/VinderTech_GliderChute/Glider_StaminaVigor_Male/FX/M_Glider_StaminaVigor_Aura
graph {
 properties {bUsedWithNiagaraSprites=true, blendMode=BLEND_Translucent, pixelDepthOffsetMode=PDOM_Legacy, shadingModel=MSM_Unlit, usedShadingModels=Unlit}
 nodes [
  Constant() {r=1}
  n1 = VectorParameter() {parameterName="Color_Morning (V3)", defaultValue=<rgba 0.3197 1 0.8698 0>}
  n2 = Constant() {r=1}
  n3 = VectorParameter() {parameterName="Color_Evening (V3)", defaultValue=<rgba 0.184 0.8091 1 0>}
  n4 = VectorParameter() {parameterName="Color_Night (V3)", defaultValue=<rgba 1 1 1 0>}
  n5 = ParticleColor()
  n6 = TextureSample(UVs=Panner(Coordinate=TextureCoordinate() {vTiling=1.5}) {speedY=0.5}) {samplerType=SAMPLERTYPE_Normal, texture=&T_Weap_Distort_N_02}
  n7 = ComponentMask(None=TextureCoordinate()) {g=true}
  n8 = Multiply(A=Multiply(A=ComponentMask(None=@n6.RGB) {g=true, r=true}, B=OneMinus(None=@n7))) {constB=0.3}
  n9 = TextureCoordinate() {uTiling=0.6, vTiling=0.4}
  n10 = ScalarParameter() {defaultValue=1, parameterName=ColorNoiseTex_UVTiling}
  n11 = MaterialFunctionCall(XY=Add(A=Multiply(A=@n9, B=@n10), B=Constant2Vector() {g=0.5})) {materialFunction=&Swizzle}
  n12 = VectorParameter() {parameterName=Color noise speed, defaultValue=<rgba 1 0 0 1>}
  n13 = TextureSampleParameter2D(UVs=Panner(Coordinate=Add(A=@n8, B=@n11.YX), Speed=ComponentMask(None=@n12.RGB) {g=true, r=true}) {speedX=1}) {parameterName=ColorNoise_Texture, samplerType=SAMPLERTYPE_LinearColor, texture=&T_NoisePack_LinearColor}
  n14 = TextureSample(UVs=TextureCoordinate() {uTiling=0.5, vTiling=0.5}) {samplerType=SAMPLERTYPE_LinearGrayscale, texture=/Engine/Functions/Engine_MaterialFunctions02/ExampleContent/Textures/LowResBlurredNoise}
  n15 = LinearInterpolate(A=@n13.G, B=@n14.R, Alpha=Constant() {r=0.5}) {constAlpha=0.2}
  n16 = SphereMask(
   A=TextureCoordinate() {vTiling=0.8}
   B=Constant2Vector() {g=0.6, r=0.5}
   Radius=ScalarParameter() {defaultValue=0.5, parameterName=SphereMask Color Radius}
  ) {attenuationRadius=0.5, hardnessPercent=0}
  n17 = ScalarParameter() {defaultValue=5, parameterName=SphereMask power}
  n18 = Multiply(
   A=LinearInterpolate(A=LinearInterpolate(Alpha=@n15), B=Constant() {r=1}, Alpha=ScalarParameter() {parameterName=Alpha})
   B=LinearInterpolate(
    A=VectorParameter() {parameterName=Color dark, defaultValue=<rgba 0 0.1257 0.4167 1>}.RGB
    B=VectorParameter() {parameterName=Color bright, defaultValue=<rgba 0.2607 1.0068 2 1>}.RGB
    Alpha=Power(Base=@n16, Exp=@n17) {constExponent=5}
   )
  )
  n19 = Panner(Coordinate=Add(A=TextureCoordinate() {uTiling=0.5, vTiling=0.5}, B=@n8) {constB=0}) {speedY=1}
  n20 = Multiply(
   A=@n18
   B=Multiply(A=TextureSample(UVs=@n19) {texture=/Game/Effects/Fort_Effects/Emotes/Energize_Max/LV4/T_Energize_Color_Noise_blur}.RGB) {constB=5}
  )
  n21 = Multiply(A=LinearInterpolate(A=Multiply(A=@n20) {constB=2}, B=@n18, Alpha=ScalarParameter() {parameterName=Use simple color mask}), B=@n5.RGB) {constA=1}
  n22 = Constant()
  n23 = MaterialFunctionCall(Value_Morning=@n22, Value_Day=@n22, Value_Evening=@n22, Value_Night=ScalarParameter() {parameterName=Night_Des}) {materialFunction=&Scalar_Time_of_Day}
  MaterialFunctionCall(Emissive=Multiply(
    A=MaterialFunctionCall(Color_Morning=@n1.RGB, Color_Day=@n2, Color_Evening=@n3.RGB, Color_Night=@n4.RGB) {materialFunction=&Color_Time_of_Day}.Result
    B=Desaturation(None=@n21, Fraction=@n23.Result)
   )) {materialFunction=&MF_FX_EmissiveAdjustments}
  n24 = Constant() {r=1}
  n25 = Multiply(A=MaterialFunctionCall() {materialFunction=&LocalPosition}.Local Position) {constB=0.01}
  n26 = Multiply(A=TextureSample(UVs=ComponentMask(None=@n25) {g=true, r=true}) {samplerType=SAMPLERTYPE_Grayscale, texture=/Game/Global/Textures/Default/Noise/LowResBlurredNoise}.RGB) {constB=0}
  n27 = VectorParameter() {parameterName=SphereMask Location, defaultValue=<rgba 0 0 0 1>}
  n28 = MaterialFunctionCall!(
   Location=TransformPosition(None=Add(A=@n26, B=@n27.RGB)) {transformType=TRANSFORMPOSSOURCE_World}
   Offset
   Radius=DynamicParameter() {defaultValue=<rgba 1 87 3 1>, paramNames [Des, SphereMask Radius, Fade edge, Appear]}.SphereMask Radius
   Calculate Camera Inside
   Depth Biased Alpha
   FadeDistance
  ) {materialFunction=/Engine/Functions/Engine_MaterialFunctions01/Gradient/SphereGradient-3D}
  n29 = Saturate(None=Power(Base=OneMinus(None=@n28.Result 0-1), Exp=ScalarParameter() {defaultValue=8, parameterName=SphereMask Hardness}) {constExponent=8})
  n30 = ScalarParameter() {defaultValue=1, parameterName=Use 3D Spheremask}
  n31 = Constant()
  n32 = TextureSampleParameter2D(UVs=Panner(Coordinate=TextureCoordinate() {uTiling=6, vTiling=0.5}) {speedY=2}) {parameterName=disturbance_Texture, samplerType=SAMPLERTYPE_Grayscale, texture=&LowResBlurredNoise}
  n33 = OneMinus(None=ComponentMask(None=TextureCoordinate()) {g=true})
  n34 = Saturate(None=Multiply(A=Power(Base=@n33) {constExponent=3.5}) {constB=20})
  n35 = ComponentMask(None=TextureCoordinate()) {g=true}
  n36 = Multiply(A=Power(Base=OneMinus(None=@n35))) {constB=50}
  n37 = Multiply(
   A=AppendVector(A=Constant(), B=Multiply(A=@n32.R) {constB=0.3})
   B=Multiply(A=LinearInterpolate(Alpha=@n34) {constA=0.1}, B=Saturate(None=@n36))
  )
  n38 = ScalarParameter() {parameterName=Use disturbance}
  n39 = TextureSampleParameter2D(UVs=Add(A=TextureCoordinate(), B=LinearInterpolate(A=@n31, B=@n37, Alpha=@n38)) {constB=0}) {
   parameterName=Default_Tex
   texture=&/T_Glider_StaminaVigorMale_Flame_01
  }
  n40 = TextureCoordinate() {vTiling=0.8}
  n41 = Constant2Vector() {g=0.6, r=0.5}
  n42 = Multiply(
   A=Multiply(
    A=StaticComponentMaskParameter(None=@n39.RGB) {defaultR=true, parameterName=Component mask}
    B=SphereMask(A=@n40, B=@n41) {attenuationRadius=0.5, hardnessPercent=0}
   )
   B=SphereMask(A=TextureCoordinate(), B=Constant2Vector() {g=0.46, r=0.5}) {attenuationRadius=0.5, hardnessPercent=50}
  )
  n43 = Multiply(A=Multiply(A=LinearInterpolate(A=@n24, B=@n29, Alpha=@n30), B=Multiply(A=@n5.A, B=@n42))) {constB=2}
  n44 = TextureCoordinate()
  n45 = VectorParameter() {parameterName=MaskTexture, defaultValue=<rgba 1 1 0 0>}
  n46 = AppendVector(A=@n45.R, B=@n45.R)
  n47 = TextureSampleParameter2D(UVs=Panner(Coordinate=Multiply(A=@n44, B=@n46), Speed=AppendVector(A=@n45.B, B=@n45.B))) {parameterName=MaskTexture, texture=/Game/Global/Textures/Default/Noise/T_SoftCells_Negative}
  n48 = Multiply(
   A=LinearInterpolate(A=@n43, B=Multiply(A=@n43, B=@n47.R), Alpha=ScalarParameter() {parameterName=Use_MaskTexture})
   B=MaterialFunctionCall("FadeSensitivity (optional)"=DynamicParameter() {defaultValue=<rgba 1 87 3 1>, paramNames [Des, SphereMask Radius, Fade edge, Appear]}.Fade edge) {materialFunction=&FadeBasedOnViewAngle}.Result
  )
  n49 = DepthFade(Opacity=@n48, FadeDistance=ScalarParameter() {defaultValue=30, parameterName=Depth fade}) {fadeDistanceDefault=50}
  n50 = TextureCoordinate()
  n51 = Add(
   A=LinearInterpolate(Alpha=ComponentMask(None=@n50) {g=true}) {constA=-0.5}
   B=LinearInterpolate(Alpha=DynamicParameter() {defaultValue=<rgba 1 87 3 1>, paramNames [Des, SphereMask Radius, Fade edge, Appear]}.Appear) {constA=-1.1, constB=2}
  )
  Multiply(A=ShadingPathSwitch(Default=@n49, Deferred=@n49, Forward=@n48, Mobile=@n48), B=Saturate(None=Multiply(A=@n51) {constB=4}))
  ScalarParameter() {defaultValue=87, parameterName=SphereMask Radius}
  ScalarParameter() {defaultValue=0.2, parameterName=A}
  ScalarParameter() {defaultValue=3, parameterName=Fade edge}
 ]
}
